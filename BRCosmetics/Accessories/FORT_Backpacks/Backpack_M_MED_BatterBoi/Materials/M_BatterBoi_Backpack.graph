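# Character-style material for the BatterBoi backpack that composites four "Dog" blink decals over the base color, between the 2023 CharacterParent PreFX and PostFX functions.
# Each decal samples a 3x3 blink flipbook at parameterized UV scale and offset, with a matching sphere mask to localize it.
# Dogs 1-3 switch frames from thresholded minimums of two time-based Frac waves, offset to desynchronize blinks.
# Dog 4's frame is a fixed parameter.
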Material /BRCosmetics/Accessories/FORT_Backpacks/Backpack_M_MED_BatterBoi/Materials/M_BatterBoi_Backpack
graph {
 properties {
  bUseMaterialAttributes=true
  bUsedWithMeshDeformer=true
  bUsedWithMorphTargets=true
  bUsedWithSkeletalMesh=true
  pixelDepthOffsetMode=PDOM_Legacy
  shadingModel=MSM_FromMaterialExpression
  usedShadingModels=Default Lit | Subsurface | Clear Coat | Subsurface Profile | Cloth
 }
 nodes [
  n1 = Sine(None=Multiply(A=Time()) {constB=0.5})
  LinearInterpolate(Alpha=Saturate(None=ConstantBiasScale(None=@n1))) {constA=0.67, constB=0.87}
  n2 = MaterialFunctionCall() {materialFunction=&MF_CharacterParent_2023_PreFX}
  n3 = GetMaterialAttributes(None=@n2.Result) {attributeGetTypes [#1]}
  n4 = NamedRerouteDeclaration(None=@n3.Base Color) {name=InBaseColor, nodeColor=<rgba 0.9219 0 1 1>}
  n5 = NamedRerouteUsage()
  n6 = NamedRerouteUsage()
  n7 = LinearInterpolate(
   A=LinearInterpolate(A=LinearInterpolate(A=@n4, B=@n5, Alpha=@n6), B=NamedRerouteUsage(), Alpha=NamedRerouteUsage())
   B=NamedRerouteUsage()
   Alpha=NamedRerouteUsage()
  )
  n8 = NamedRerouteUsage()
  n9 = NamedRerouteUsage()
  MaterialFunctionCall(
   Material=SetMaterialAttributes!(MaterialAttributes=@n3.MaterialAttributes, Base Color=LinearInterpolate(A=@n7, B=@n8, Alpha=@n9)) {attributeSetTypes [#1]}
   M=@n2.M
   Pre Skinned Local Position=@n2.M
   Distance Blend=@n2.M
   Fresnel=@n2.M
  ) {materialFunction=&MF_CharacterParent_2023_PostFX}
  n10 = Constant() {r=3}
  n11 = TextureCoordinate()
  n12 = ScalarParameter() {defaultValue=0.11, parameterName=DOg1_ScaLe}
  n13 = ScalarParameter() {defaultValue=-2.53, parameterName=Dog1U}
  n14 = ScalarParameter() {defaultValue=0.662, parameterName=Dog1V}
  n15 = MaterialFunctionCall(
   "Animation  Phase (0-1)"=LinearInterpolate(Alpha=NamedRerouteUsage()) {constA=0.2, constB=0}
   Number of Rows=@n10
   Number of Columns=@n10
   Texture=TextureObject() {texture=/BRCosmetics/Accessories/FORT_Backpacks/Backpack_M_MED_BatterBoi/Textures/T_Blink_FlipBook}
   UVs=Add(
    A=MaterialFunctionCall(UVs=@n11, Texture Scale=@n12) {materialFunction=&ScaleUVsByCenter}.UVs
    B=AppendVector(A=@n13, B=@n14)
   )
  ) {materialFunction=&FlipBook}
  NamedRerouteDeclaration(None=LinearInterpolate(A=NamedRerouteUsage(), B=@n15.Result, Alpha=@n15.Result)) {name=Dog1, nodeColor=<rgba 0 1 0.6016 1>}
  n16 = Add(A=Time()) {constB=0.5}
  n17 = Constant() {r=0.3}
  n18 = Constant() {r=0.17}
  n19 = Min(A=Frac(None=Multiply(A=@n16, B=@n17)), B=Frac(None=Multiply(A=@n16, B=@n18) {constB=0.17}))
  n20 = ScalarParameter() {defaultValue=0.06, parameterName=BlinkTime}
  n21 = Constant()
  n22 = Constant() {r=1}
  n23 = Constant() {r=3}
  n24 = TextureCoordinate()
  n25 = ScalarParameter() {defaultValue=0.1, parameterName=DOg2_ScaLe}
  n26 = ScalarParameter() {defaultValue=1.7392, parameterName=Dog2U}
  n27 = ScalarParameter() {defaultValue=1.127, parameterName=Dog2V}
  n28 = MaterialFunctionCall(
   "Animation  Phase (0-1)"=LinearInterpolate(Alpha=If(A=@n19, B=@n20, "A > B"=@n21, "A < B"=@n22)) {constA=0.35, constB=0.54}
   Number of Rows=@n23
   Number of Columns=@n23
   Texture=TextureObject() {texture=&T_Blink_FlipBook}
   UVs=Add(
    A=MaterialFunctionCall(UVs=@n24, Texture Scale=@n25) {materialFunction=&ScaleUVsByCenter}.UVs
    B=AppendVector(A=@n26, B=@n27)
   )
  ) {materialFunction=&FlipBook}
  NamedRerouteDeclaration(None=LinearInterpolate(A=NamedRerouteUsage(), B=@n28.Result, Alpha=@n28.Result)) {name=Dog2, nodeColor=<rgba 0 1 0.6016 1>}
  n29 = ScalarParameter() {defaultValue=0.9238, parameterName=Dog2_MaskU}
  n30 = ScalarParameter() {defaultValue=0.1, parameterName=Dog2_MaskV}
  NamedRerouteDeclaration(None=SphereMask(A=TextureCoordinate(), B=AppendVector(A=@n29, B=@n30), Radius=ScalarParameter() {defaultValue=0.056, parameterName=Dog2Radius})) {name=Dog2Mask, nodeColor=<rgba 1 0.3516 0 1>}
  n31 = Add(A=Time()) {constB=0.8}
  n32 = Constant() {r=0.3}
  n33 = Constant() {r=0.17}
  n34 = Min(A=Frac(None=Multiply(A=@n31, B=@n32)), B=Frac(None=Multiply(A=@n31, B=@n33) {constB=0.17}))
  n35 = ScalarParameter() {defaultValue=0.06, parameterName=BlinkTime}
  n36 = Constant()
  n37 = Constant() {r=1}
  n38 = Constant() {r=3}
  n39 = TextureCoordinate()
  n40 = ScalarParameter() {defaultValue=0.1, parameterName=DOg3_ScaLe}
  n41 = ScalarParameter() {defaultValue=-0.0406, parameterName=Dog3U}
  n42 = ScalarParameter() {defaultValue=1.3112, parameterName=Dog3V}
  n43 = MaterialFunctionCall(
   "Animation  Phase (0-1)"=LinearInterpolate(Alpha=If(A=@n34, B=@n35, "A > B"=@n36, "A < B"=@n37)) {constA=0.67, constB=0.87}
   Number of Rows=@n38
   Number of Columns=@n38
   Texture=TextureObject() {texture=&T_Blink_FlipBook}
   UVs=Add(
    A=MaterialFunctionCall(UVs=@n39, Texture Scale=@n40) {materialFunction=&ScaleUVsByCenter}.UVs
    B=AppendVector(A=@n41, B=@n42)
   )
  ) {materialFunction=&FlipBook}
  NamedRerouteDeclaration(None=LinearInterpolate(A=NamedRerouteUsage(), B=@n43.Result, Alpha=@n43.Result)) {name=Dog3, nodeColor=<rgba 0 1 0.6016 1>}
  n44 = ScalarParameter() {defaultValue=0.5, parameterName=Dog3_MaskU}
  n45 = ScalarParameter() {defaultValue=0.076, parameterName=Dog3_MaskV}
  NamedRerouteDeclaration(None=SphereMask(A=TextureCoordinate(), B=AppendVector(A=@n44, B=@n45), Radius=ScalarParameter() {defaultValue=0.04, parameterName=Dog3Radius})) {name=Dog3Mask_1, nodeColor=<rgba 1 0.3516 0 1>}
  n46 = Constant() {r=3}
  n47 = TextureCoordinate()
  n48 = ScalarParameter() {defaultValue=0.095, parameterName=DOg4_ScaLe}
  n49 = ScalarParameter() {defaultValue=3.6062, parameterName=Dog4U}
  n50 = ScalarParameter() {defaultValue=-0.6063, parameterName=Dog4V}
  n51 = MaterialFunctionCall(
   "Animation  Phase (0-1)"=ScalarParameter() {defaultValue=0.77, parameterName=Dog4}
   Number of Rows=@n46
   Number of Columns=@n46
   Texture=TextureObject() {texture=&T_Blink_FlipBook}
   UVs=Add(
    A=MaterialFunctionCall(UVs=@n47, Texture Scale=@n48) {materialFunction=&ScaleUVsByCenter}.UVs
    B=AppendVector(A=@n49, B=@n50)
   )
  ) {materialFunction=&FlipBook}
  NamedRerouteDeclaration(None=LinearInterpolate(A=NamedRerouteUsage(), B=@n51.Result, Alpha=@n51.Result)) {name=Dog4, nodeColor=<rgba 0 1 0.6016 1>}
  n52 = ScalarParameter() {defaultValue=0.6333, parameterName=Dog4_MaskU}
  n53 = ScalarParameter() {defaultValue=0.076, parameterName=Dog4_MaskV}
  NamedRerouteDeclaration(None=SphereMask(A=TextureCoordinate(), B=AppendVector(A=@n52, B=@n53), Radius=ScalarParameter() {defaultValue=0.048, parameterName=Dog4Radius})) {name=Dog4_Mask, nodeColor=<rgba 1 0.3516 0 1>}
  n54 = ScalarParameter() {defaultValue=0.7778, parameterName=Dog1_MaskU}
  n55 = ScalarParameter() {defaultValue=0.1, parameterName=Dog1_MaskV}
  NamedRerouteDeclaration(None=SphereMask(A=TextureCoordinate(), B=AppendVector(A=@n54, B=@n55), Radius=ScalarParameter() {defaultValue=0.05, parameterName=Dog1Radius})) {name=Dog1Mask, nodeColor=<rgba 1 0.3516 0 1>}
  n56 = Time()
  n57 = Frac(None=Multiply(A=@n56, B=Constant() {r=0.3}))
  n58 = Frac(None=Multiply(A=@n56, B=Constant() {r=0.17}) {constB=0.17})
  NamedRerouteDeclaration(None=If(A=Min(A=@n57, B=@n58), B=ScalarParameter() {defaultValue=0.06, parameterName=BlinkTime}, "A > B"=Constant(), "A < B"=Constant() {r=1})) {name=Blinking, nodeColor=<rgba 0 1 0.5547 1>}
  ScalarParameter() {defaultValue=0.45, parameterName=Paramddawf}
  ScalarParameter() {defaultValue=0.77, parameterName=Paramddawf3}
 ]
}
